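; NitroDrome FX manager that fires arena reaction effects: screens, spotlights, sparks and head FX.
; Multicast events trigger spotlights, sparks, or screen reactions with a random byte.
; Screens map that byte to a texture, writing time and byte to custom primitive data; the boss screen function is empty.
; Sparks gather emitter transforms into arrays feeding one cosmetic-only Niagara system, with audio triggered or spawned per emitter.
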(event Custom|Multicast_Activate_RustBucketBossScreens
  (CallFunction|ActivateRustBucketBossScreens (Math|Conversions|ToByte(Integer) (Math|Random|RandomIntegerinRange 0 4))))

(event Custom|ActivateRustBucketBossScreens
  (params (Byte%20In :type Byte))
  (Default|ActivateBossScreensFunc Byte%20In))

(event Custom|ActivateRustBucketScreens
  (params (Byte%20In :type Byte))
  (Default|ActivateScreens Byte%20In))

(event Custom|Multicast_Activate_RustBucketScreens
  (CallFunction|ActivateRustBucketScreens (Math|Conversions|ToByte(Integer) (Math|Random|RandomIntegerinRange 0 4))))

(event Custom|Multicast_Activate_Spotlights
  (CallFunction|ActivateSpotlights))

(event Custom|Multicast_Activate_Sparks
  (CallFunction|ActivateSparks))

(event Custom|ActivateSpotlights
  (for _array_element (Variables|Default|GetLightActors)
    (Utilities|IsValid (Utilities|ResolveSoftReference _array_element)
      (:Is%20Valid
        (Components|Activation|Activate (Variables|/Game/Environments/Helios/Props/RustBucket/RustBucket_SpotLight_A/Blueprints/BP_RustBucket_NoSpotLight_A_NitroDrome.BP_RustBucket_NoSpotLight_A_NitroDrome_C|GetNS_NitroDrome_Reaction_OneSpot (Utilities|ResolveSoftReference _array_element)) true)))))

(event Custom|ActivateEffects
  (CallFunction|ActivateSpotlights)
  (CallFunction|ActivateSparks)
  (CallFunction|ActivateHeadFX))

(event Custom|ActivateHeadFX
  (bind _headfx (Variables|Default|GetHeadFX))
  (bind _output (Utilities|ResolveSoftReference _headfx))
  (Utilities|IsValid (Utilities|ResolveSoftReference (Variables|Default|GetEntranceHead))
    (:Is%20Valid
      (Utilities|IsValid _output
        (:Is%20Valid
          (Components|Activation|Activate (Variables|/Script/Engine.Actor|GetRootComponent _output) true))))))

(event Custom|ActivateSparks
  (Utilities|Array|Clear (Variables|Default|GetSpark_Pot_Positions))
  (Utilities|Array|Clear (Variables|Default|GetSpark_Pot_Quaternions))
  (for _array_element (Variables|Default|GetSparksActors)
    (Utilities|IsValid (Utilities|ResolveSoftReference _array_element)
      (:Is%20Valid
        (bind (_returnvalue_location _returnvalue_rotation) (Math|Transform|BreakTransform (Transformation|GetActorTransform (Utilities|ResolveSoftReference _array_element))))
        (Utilities|Array|AddUnique (Variables|Default|GetSpark_Pot_Positions) _returnvalue_location)
        (Utilities|Array|AddUnique (Variables|Default|GetSpark_Pot_Quaternions) (Math|Conversions|ToQuaternion(Rotator) _returnvalue_rotation))
        (CallFunction|ActivateAudio (Utilities|ResolveSoftReference _array_element)))))
  (Networking|CanExecuteCosmeticEvents
    (:True
      (bind _returnvalue (Niagara|SpawnSystematLocation :SystemTemplate "/Game/Environments/Helios/FX/NitroDrome/FX/NS_NitroDrome_Reaction_Spot_Sparks_Complex.NS_NitroDrome_Reaction_Spot_Sparks_Complex" :Location (Transformation|GetActorLocation) :Scale "1.000000,1.000000,1.000000" :bAutoDestroy true :bAutoActivate true :PoolingMethod "AutoRelease" :bPreCullCheck false))
      (CallFunction|UpdateNIagaraParams _returnvalue _returnvalue))))

(fn ActivateAudio
  (params (SparkLocation :type Temp%20Spark%20Fx%20Emitter%20Object%20Reference))
  (bind _activesparks (Variables|/Game/Athena/Asteria/Maps/Test/Dev/Mike/temp_spark_fx_emitter.temp_spark_fx_emitter_C|GetActiveSparks SparkLocation))
  (Utilities|IsValid _activesparks
    (:Is%20Valid
      (Audio|Parameter|ExecuteTriggerParameter _activesparks "NewEvent"))
    (:Is%20Not%20Valid
      (bind _returnvalue (Audio|SpawnSoundatLocation "/Game/Sounds/Helios/POI/NitroDrome/Nitrodrome_Sparks_Cue.Nitrodrome_Sparks_Cue" (Transformation|GetWorldLocation (Variables|/Game/Athena/Asteria/Maps/Test/Dev/Mike/temp_spark_fx_emitter.temp_spark_fx_emitter_C|GetSparkLocation SparkLocation))))
      (Variables|/Game/Athena/Asteria/Maps/Test/Dev/Mike/temp_spark_fx_emitter.temp_spark_fx_emitter_C|SetActiveSparks _returnvalue SparkLocation))))

(fn ActivateBossScreensFunc
  (params (Byte :type Byte)))

(fn ActivateScreens
  (params (Byte :type Byte))
  ;[0.04739 0.3333 0.006409] 0 - NitroDrome.Event.EnteredDrome.Vehicle
  ; 1 - NitroDrome.Event.Reward.Approaching
  ; 2 - NitroDrome.Event.LootRewardSpawned
  ; 3 - NitroDrome.Event.VehicleEliminated
  ; 4 - NitroDrome.Event.PawnEliminated.Player
  ; 5 - NitroDrome.Event.Boss.Spawn
  (for _array_element (Variables|Screens|GetScreenActors)
    (Utilities|IsValid (Utilities|ResolveSoftReference _array_element)
      (:Is%20Valid
        (Default|SetTextureParamPrimitiveDataandColor (Variables|/Script/FortniteGame.BuildingSMActor|GetStaticMeshComponent (Utilities|ResolveSoftReference _array_element)) (Variables|GetByte))))))

(fn OnRep_Spark_Pot_Positions)

(fn Set_BossScreen_TextureParam_PrimitiveData_And_Color
  (params (Screen :type Base%20Building%20Static%20Mesh%20Component%20Object%20Reference) (Btye :type Byte))
  (bind _bytetexturemap (Variables|Screens|GetByteTextureMap))
  (bind (_value _returnvalue) (Utilities|Map|Find _bytetexturemap Btye))
  ;[0.04739 0.3333 0.006409] 0 - NitroDrome.Event.EnteredDrome.Vehicle
  ; 1 - NitroDrome.Event.Reward.Approaching
  ; 2 - NitroDrome.Event.LootRewardSpawned
  ; 3 - NitroDrome.Event.VehicleEliminated
  ; 4 - NitroDrome.Event.PawnEliminated.Player
  ; 5 - NitroDrome.Event.Boss.Spawn
  (if _returnvalue
    (Rendering|Material|SetTextureParameterValue (Class|BTechArtGlobalLibrary|GetDynamicMaterialInstance Screen 1) "ReactionTexture" (Utilities|ResolveSoftReference _value))
    (Rendering|Material|SetCustomPrimitiveDataFloat Screen 1 (Utilities|Time|GetGameTimeinSeconds))
    (Rendering|Material|SetCustomPrimitiveDataFloat Screen 2 (Math|Conversions|ToFloat(Byte) Btye))))

(fn Set_TextureParam_PrimitiveData_And_Color
  (params (Screen :type Base%20Building%20Static%20Mesh%20Component%20Object%20Reference) (Btye :type Byte))
  (bind _bytetexturemap (Variables|Screens|GetByteTextureMap))
  (bind (_value _returnvalue) (Utilities|Map|Find _bytetexturemap Btye))
  ;[0.04739 0.3333 0.006409] 0 - NitroDrome.Event.EnteredDrome.Vehicle
  ; 1 - NitroDrome.Event.Reward.Approaching
  ; 2 - NitroDrome.Event.LootRewardSpawned
  ; 3 - NitroDrome.Event.VehicleEliminated
  ; 4 - NitroDrome.Event.PawnEliminated.Player
  ; 5 - NitroDrome.Event.Boss.Spawn
  (if _returnvalue
    (Rendering|Material|SetTextureParameterValue (Class|BTechArtGlobalLibrary|GetDynamicMaterialInstance Screen 1) "ReactionTexture" (Utilities|ResolveSoftReference _value))
    (Rendering|Material|SetCustomPrimitiveDataFloat Screen 1 (Utilities|Time|GetGameTimeinSeconds))
    (Rendering|Material|SetCustomPrimitiveDataFloat Screen 2 (Math|Conversions|ToFloat(Byte) Btye))))

(fn UpdateNIagaraParams
  (params (InputPin :type Niagara%20Particle%20System%20Component%20Object%20Reference) (NiagaraSystem :type Niagara%20Particle%20System%20Component%20Object%20Reference))
  (Niagara|NiagaraSetPositionArray NiagaraSystem "AvailablePot_Locations" (Variables|Default|GetSpark_Pot_Positions))
  (Niagara|NiagaraSetQuaternionArray InputPin "AvailablePot_Rotations" (Variables|Default|GetSpark_Pot_Quaternions))
  (Components|Activation|Activate InputPin true))

(fn ConstructionScript)
