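# Looping ambient Niagara effect for the NitroDrome entrance head, emitting faint magenta smoke from its mouth as GPU sprites.
# Particles spawn in a small cylinder and launch along two diagonal cones chosen by partition, colored between User.ColorMin and ColorMax.
# Emitter world Z is remapped to shrink speed and sprite size at greater height.
# Flipbook animation, fade curves, time-of-day tint and gravity shape each puff; collision is disabled.
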NiagaraSystem /Game/Environments/Helios/FX/NitroDrome/FX/NS_NitroDrome_EntranceHead_Ambient_Loop
dependencies {
 usedDataInterfaces [&NiagaraDataInterfaceArrayFloat3, &NiagaraDataInterfaceCurve]
 usedDynamicInputs []
 usedModules [&SystemState]
}
emitters {Smoke_Mouth001 {
  simTarget=GPUComputeSim
  emitterUpdateScript [
   &EmitterState()
   &SpawnRate(SpawnRate=&Multiply_Float(A=5, B=1), Use Spawn Probability=false)
  ]
  particleSpawnScript [
   &InitializeParticle(
    Color Maximum=$User.ColorMax
    Color Minimum=$User.ColorMin
    Color Mode=::NewEnumerator2
    Lifetime Max=4
    Lifetime Min=3
    Lifetime Mode=::NewEnumerator1
    Sprite Rotation Mode=::NewEnumerator2
    Sprite Size Mode=::NewEnumerator1
    Uniform Sprite Size Max=225
    Uniform Sprite Size Min=150
    UsePositionOffset=false
    Write Material Random=true
   ) {module=InitializeParticle001}
   &PartitionParticles()
   /SetVariables(Particles.RandomFrame=&RandomRangeInt(Minimum=0, Maximum=63)) {isSetParameters=true}
   &DynamicMaterialParameters(
    Index 0 Param 1=&RandomRangeFloat(Minimum=0.2, Maximum=0.5)
    Module.Param0WriteEnabled=true
    Module.Param1WriteEnabled=true
    Module.Param2WriteEnabled=true
    Module.Param3WriteEnabled=true
   )
   &ShapeLocation(
    Cylinder Height=25
    Cylinder Radius=25
    MaskSpawnGroup=false
    Override Seed=false
    Shape Primitive=::NewEnumerator1
    Surface Only=false
    UseOwnerScale=true
   )
   &CameraOffset(Camera Offset Amount=&RandomRangeFloat(Minimum=2, Maximum=5))
   &SystemLocation(MaskSpawnGroup=false)
   &AddVelocity(
    Cone Angle=15
    Cone Axis=&SelectVectorFromArray(
     Array Sampling Mode=::NewEnumerator0
     Direct Array Index=$Particles.PartitionParticles.Partition
     Vector Selection Array=DataInterface() {propertyValues {FloatData [<1 1 0>, <1 -1 0>]}}
     Random Seed=0
    )
    Inner Cone Angle=5
    Module.UseConeFalloff=false
    Module.UseVelDistribution=false
    Override Seed=false
    Velocity Mode=::NewEnumerator2
    Velocity Speed=&Multiply_Float(
     A=&RandomRangeFloat(Minimum=350, Maximum=500)
     B=&Multiply_Float(
      A=1
      B=&Lerp_Float(
       A=1
       B=0.5
       Alpha=&RemapRange(
        Input Value To Remap=&MakeFloatFromVector(
         Vector=&TransformPosition(Position=$Engine.Emitter.SimulationPosition, Source Space=::Local, Destination Space=::World)
         Channel=::NewEnumerator2
        )
        Input Min=3456
        Input Max=4556.1875
        Output Min=0
        Output Max=1
       )
      )
     )
    )
   )
  ]
  particleUpdateScript [
   &ParticleState()
   &FN_SubUVAnimation(
    End Frame=&Add_Integer(A=0, B=63)
    SubUV Animation Mode:ENiagara_SubUVLookupMode=::NewEnumerator1
    SubUV Curve=DataInterface() {propertyValues {Curve {keys [
        {arriveTangent=0.7286, interpMode=RCIM_Cubic, leaveTangent=0.7286, tangentMode=RCTM_User, time=0, value=0}
        {arriveTangent=0.5299, tangentMode=RCTM_User, time=1, value=0.9}
       ]}}}
    SubUV Loop Count:NiagaraFloat=&RandomRangeFloat(Minimum=1, Maximum=1)
   )
   &ScaleSpriteSize(Scale Factor=&Vector2DFromFloat(Value=&FloatFromCurve(
      FloatCurve=DataInterface() {propertyValues {Curve {keys [
          {arriveTangent=2.5076, interpMode=RCIM_Cubic, leaveTangent=2.5076, tangentMode=RCTM_User, time=0.0023, value=0}
          {arriveTangent=0.0897, interpMode=RCIM_Cubic, leaveTangent=0.0897, tangentMode=RCTM_User, time=1, value=1}
         ]}}}
      Scale Curve=&Multiply_Float(
       A=5
       B=&Lerp_Float(
        A=1
        B=0.4
        Alpha=&RemapRange(
         Input Value To Remap=&MakeFloatFromVector(
          Vector=&TransformPosition(Position=$Engine.Emitter.SimulationPosition, Source Space=::Local, Destination Space=::World)
          Channel=::NewEnumerator2
         )
         Input Min=3456
         Input Max=4556.1875
         Output Min=0
         Output Max=1
        )
       )
      )
     )))
   &ScaleColor(
    Scale Alpha=&FloatFromCurve(
     FloatCurve=DataInterface() {propertyValues {Curve {keys [
         {interpMode=RCIM_Cubic, tangentMode=RCTM_User, time=0, value=0}
         {interpMode=RCIM_Cubic, tangentMode=RCTM_User, time=0.3, value=1.05}
         {arriveTangent=-1.5, interpMode=RCIM_Cubic, leaveTangent=-1.5, time=0.9, value=1.05}
         {tangentMode=RCTM_User, time=1, value=0}
        ]}}}
     Scale Curve=&Multiply_Float(A=0.15, B=0.2)
    )
    Scale RGB=&Multiply_VectorByFloat(
     Vector=<1 1 1>
     Float=&FloatFromCurve(
      FloatCurve=DataInterface() {propertyValues {Curve {keys [
          {interpMode=RCIM_Cubic, tangentMode=RCTM_User, time=0, value=0.5}
          {arriveTangent=0.0148, interpMode=RCIM_Cubic, leaveTangent=0.0148, tangentMode=RCTM_User, time=0.2, value=1.35}
          {arriveTangent=-1.6875, interpMode=RCIM_Cubic, leaveTangent=-1.6875, time=0.8, value=1}
          <at 1 0>
         ]}}}
      Scale Curve=4
     )
    )
    ScaleA=true
    ScaleRGB=true
   ) {module=ScaleColor001}
   &DynamicMaterialParameters(
    Index 0 Param 1=&FloatFromCurve(
     FloatCurve=DataInterface() {propertyValues {Curve {keys [
         {tangentMode=RCTM_User, time=0, value=0}
         {interpMode=RCIM_Cubic, tangentMode=RCTM_User, time=0.8, value=0.125}
         {tangentMode=RCTM_User, time=1, value=1}
        ]}}}
     Scale Curve=0
    )
    Index 0 Param 3=1
    Module.Index0Param0Write=true
    Module.Index0Param1Write=true
    Module.Index0Param2Write=true
    Module.Index0Param3Write=true
   ) {module=DynamicMaterialParameters001}
   &self:ColorFromTOD_11(
    Afternoon:LinearColor=<rgba 1 0.9954 0.9249 1>
    Evening:LinearColor=<rgba 0.989 1 0.7298 1>
    Morning:LinearColor=<rgba 1 0.9526 0.8666 1>
    Night:LinearColor=<rgba 0.7 0.1065 0.5763 1>
    TOD:Vector4f=&MakeVector4_FromLinearColor(LinearColor=$NPC.FortniteNPC.FortniteActiveTimeOfDay)
   ) {module=ColorFromTOD_11}
   &GravityForce(Gravity=<0 0 -150>)
   &Collision(
    Age Colliding Particles=false
    Friction=0
    Kill Offscreen Particles=true
    KillOnCollision=false
    Module.Control Rest State Via Surface Normal=false
    Module.EnableMaxCollisionCount=false
    Module.Randomize Collision Normal Vector=false
    Restitution=0
   ) {enabled=false}
   &SolveForcesAndVelocity(Clamp Velocity=true, Limit Acceleration=false, Speed Limit=250)
  ]
  rendererClasses [&NiagaraSpriteRendererProperties]
  renderers [{rendererClass=&NiagaraSpriteRendererProperties}]
 }}
summary {
 systemName=NS_NitroDrome_EntranceHead_Ambient_Loop
 emitters [{emitterName=Smoke_Mouth001, simTarget=GPUComputeSim, rendererClasses [&NiagaraSpriteRendererProperties]}]
 userVariables [
  User.ColorMax {type=LinearColor, description="", defaultValue {struct=LinearColor, value=<rgba 1 0 0.6842 1>}}
  User.ColorMin {type=LinearColor, description="", defaultValue {struct=LinearColor, value=<rgba 0.5003 0 0.6042 1>}}
 ]
}
