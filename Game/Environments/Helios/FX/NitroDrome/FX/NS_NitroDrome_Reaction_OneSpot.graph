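# Niagara GPU burst effect for the NitroDrome environment, a short-lived glowing orange "reaction" that spawns at a single spot.
# FN_Sprite001 draws sprites that grow while briefly drifting on curl noise; FN_Sprite002 draws meshes with sine-driven rotation wobble.
# Both fade in and out, boost color for glow, offset toward camera, and feed distance-remapped values into dynamic material parameters, with distance also scaling meshes.
# The FN_Sprite emitter is disabled and has no renderer.
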NiagaraSystem /Game/Environments/Helios/FX/NitroDrome/FX/NS_NitroDrome_Reaction_OneSpot
dependencies {
 usedDataInterfaces [&NiagaraDataInterfaceColorCurve, &NiagaraDataInterfaceCurve]
 usedDynamicInputs [&Lerp_Float]
 usedModules [
  &SystemState
  &self:SystemSpawnScript.SystemScriptSource.SystemScriptGraph.NiagaraNodeAssignment_0.SetVariables_#1
 ]
}
emitters {
 FN_Sprite {
  bEnabled=false
  simTarget=GPUComputeSim
  emitterUpdateScript [
   &EmitterState()
   &SpawnBurst_Instantaneous(Use Spawn Probability=false, UseLoopCountLimit=false)
  ]
  particleSpawnScript [
   &InitializeParticle(
    Color=&ColorFromCurve(
     ColorCurve=DataInterface() {propertyValues {
       AlphaCurve {keys [<at 0 1 RCIM_Cubic>,<at 0.9977 1 RCIM_Cubic>]}
       BlueCurve {keys [<at 0.1175 0.376 RCIM_Cubic>]}
       GreenCurve {keys [<at 0.1175 0 RCIM_Cubic>]}
       RedCurve {keys [<at 0.1175 1 RCIM_Cubic>]}
      }}
     CurveIndex=&RandomRangeFloat(Minimum=0, Maximum=1)
    )
    Lifetime Max=3.5
    Lifetime Min=3.5
    Lifetime Mode=::NewEnumerator1
    Mesh Scale Mode=::NewEnumerator1
    Mesh Uniform Scale Max=1
    Sprite Size Mode=::NewEnumerator1
    Uniform Sprite Size Max=1
    Uniform Sprite Size Min=1
    UsePositionOffset=false
    Write Material Random=true
   ) {module=InitializeParticle001}
   &AddVelocityFromPoint(
    Module.Use Falloff Distance=false
    Origin Offset=&UniformRangedVector(Minimum=<0 0 -100>, Maximum=<0 0 -100>)
    Origin Offset Coordinate Space=::Simulation
    Velocity Strength=&UniformRangedFloat(Minimum=10, Maximum=15)
   ) {enabled=false}
   /Niagara/Modules/Update/Forces/CurlNoiseForce(Noise Frequency=0.25, Noise Strength=5, Pan Noise=true, Pan Noise Field=<0 0 -5>) {enabled=false, module=CurlNoiseForce001}
   &ApplyInitialForces() {enabled=false}
   &InitialMeshOrientation(
    Mesh Axis To Orient=<0 0 0>
    Mesh Orientation Mode=::NewEnumerator2
    Orientation Vector=<0 0 1>
    Rotation=&RandomRangeVector(Minimum=<0 0 0>, Maximum=<0 0 0>)
    Use Rotation Vector=true
   )
   &ShapeLocation(
    Cylinder Height=2
    Cylinder Radius=&FN_LOD_Distance_Remap(Start Distance=5000, End Distance=15000, Minimum Output=6000, Maximum Output=3000, Clamp Output=true)
    MaskSpawnGroup=false
    Non Uniform Scale=<2 1 1>
    Offset=<0 0 1500>
    Offset Mode=::NewEnumerator0
    Override Seed=false
    Shape Primitive=::NewEnumerator1
    Surface Only=true
    Use Endcaps In Surface Only Mode=false
    UseOwnerScale=true
   ) {enabled=false}
  ]
  particleUpdateScript [
   &ParticleState()
   &UpdateMeshOrientation() {enabled=false, module=UpdateMeshOrientation001}
   &UpdateMeshOrientation(
    Facing Interpolation=false
    Orientation Method=::NewEnumerator1
    Override Quat=false
    TransformFacing=false
    TransformUp=false
    Up Direction:Vector3f=<0 0 -1>
   )
   &SolveForcesAndVelocity(Clamp Velocity=false, Limit Acceleration=false)
   &ScaleColor(
    Scale Alpha=&FloatFromCurve(FloatCurve=DataInterface() {propertyValues {Curve {keys [
         {arriveTangent=9.0725, interpMode=RCIM_Cubic, leaveTangent=9.0725, tangentMode=RCTM_User, time=0, value=0}
         {arriveTangent=-2.5249, interpMode=RCIM_Cubic, leaveTangent=-2.5249, tangentMode=RCTM_User, time=0.4, value=1.075}
         {arriveTangent=-0.0405, interpMode=RCIM_Cubic, leaveTangent=-0.0405, tangentMode=RCTM_User, time=1, value=0}
        ]}}})
    ScaleA=true
    ScaleRGB=true
   )
  ]
  rendererClasses []
  renderers []
 }
 FN_Sprite001 {
  simTarget=GPUComputeSim
  emitterUpdateScript [
   &EmitterState()
   &SpawnBurst_Instantaneous(Use Spawn Probability=false, UseLoopCountLimit=false)
  ]
  particleSpawnScript [
   &InitializeParticle(
    Color=&ColorFromCurve(
     ColorCurve=DataInterface() {propertyValues {
       AlphaCurve {keys [<at 0 1 RCIM_Cubic>,<at 0.9977 1 RCIM_Cubic>]}
       BlueCurve {keys [<at 0.1175 0 RCIM_Cubic>]}
       GreenCurve {keys [<at 0.1175 0.1279 RCIM_Cubic>]}
       RedCurve {keys [<at 0.1175 1 RCIM_Cubic>]}
      }}
     CurveIndex=&RandomRangeFloat(Minimum=0, Maximum=1)
    )
    Lifetime Max=3
    Lifetime Min=3
    Lifetime Mode=::NewEnumerator1
    Mesh Scale Mode=::NewEnumerator1
    Mesh Uniform Scale Max=250
    Mesh Uniform Scale Min=50
    Sprite Size Mode=::NewEnumerator1
    Uniform Sprite Size Max=300
    Uniform Sprite Size Min=250
    UsePositionOffset=false
    Write Material Random=true
   ) {module=InitializeParticle001}
   &AddVelocityFromPoint(
    Module.Use Falloff Distance=false
    Origin Offset=&UniformRangedVector(Minimum=<0 0 -100>, Maximum=<0 0 -100>)
    Origin Offset Coordinate Space=::Simulation
    Velocity Strength=&UniformRangedFloat(Minimum=10, Maximum=15)
   ) {enabled=false}
   &CurlNoiseForce(Noise Frequency=0.25, Noise Strength=5, Pan Noise=true, Pan Noise Field=<0 0 -5>) {enabled=false, module=CurlNoiseForce001}
   &ApplyInitialForces() {enabled=false}
   &InitialMeshOrientation(
    Mesh Axis To Orient=<0 0 0>
    Mesh Orientation Mode=::NewEnumerator2
    Orientation Vector=<0 0 1>
    Rotation=&RandomRangeVector(Minimum=<0 0 0>, Maximum=<0 0 0>)
    Use Rotation Vector=true
   )
   &ShapeLocation(
    Cylinder Height=2
    Cylinder Radius=&FN_LOD_Distance_Remap(Start Distance=5000, End Distance=15000, Minimum Output=6000, Maximum Output=3000, Clamp Output=true)
    MaskSpawnGroup=false
    Non Uniform Scale=<2 1 1>
    Offset=<0 0 1500>
    Offset Mode=::NewEnumerator0
    Override Seed=false
    Shape Primitive=::NewEnumerator1
    Surface Only=true
    Use Endcaps In Surface Only Mode=false
    UseOwnerScale=true
   ) {enabled=false}
  ]
  particleUpdateScript [
   &ParticleState()
   &UpdateMeshOrientation() {enabled=false, module=UpdateMeshOrientation001}
   &ScaleMeshSize(Scale Factor=&Multiply_VectorByFloat(
     Vector=<0.25 0.25 0.25>
     Float=&FN_LOD_Distance_Remap(Start Distance=5000, End Distance=50000, Minimum Output=0.5, Maximum Output=2, Clamp Output=true)
    ))
   &UpdateMeshOrientation(
    Facing Interpolation=false
    Orientation Method=::NewEnumerator1
    Override Quat=false
    TransformFacing=false
    TransformUp=false
    Up Direction:Vector3f=<0 0 -1>
   )
   &UpdateMeshOrientation(
    Rotation Rate=2
    Rotation Vector=&MakeVector(
     X=0
     Y=&Sine(Normalized Angle=$Emitter.Age, Period=&RandomRangeFloat(Minimum=2, Maximum=3), Scale=0.02, Bias=0)
     Z=0
    )
   ) {module=UpdateMeshOrientation002}
   &CurlNoiseForce(
    Noise Frequency=1
    Noise Quality / Cost=::NewEnumerator0
    Noise Strength=&FloatFromCurve(FloatCurve=DataInterface() {propertyValues {Curve {keys [<at 0 1>,<at 0.25 0>]}}})
    Pan Noise=true
    Pan Noise Field=<0 10 0>
   )
   &ScaleSpriteSize(Scale Factor=&Vector2DFromFloat(Value=&FloatFromCurve(
      FloatCurve=DataInterface() {propertyValues {Curve {keys [{arriveTangent=4.5137, interpMode=RCIM_Cubic, leaveTangent=4.5137, tangentMode=RCTM_User, time=0, value=0}, <at 1 0.5 RCIM_Cubic>]}}}
      Scale Curve=2
     )))
   &SolveForcesAndVelocity(Clamp Velocity=false, Limit Acceleration=false)
   &ScaleColor(
    Scale Alpha=&FloatFromCurve(FloatCurve=DataInterface() {propertyValues {Curve {keys [
         {arriveTangent=9.0725, interpMode=RCIM_Cubic, leaveTangent=9.0725, tangentMode=RCTM_User, time=0, value=0}
         {arriveTangent=-2.5249, interpMode=RCIM_Cubic, leaveTangent=-2.5249, tangentMode=RCTM_User, time=0.4, value=1.075}
         {arriveTangent=-0.0405, interpMode=RCIM_Cubic, leaveTangent=-0.0405, tangentMode=RCTM_User, time=1, value=0}
        ]}}})
    ScaleA=true
    ScaleRGB=true
   )
   &DynamicMaterialParameters(
    Index 0 Param 1=&Multiply_Float(
     A=25
     B=&FN_LOD_Distance_Remap(Start Distance=250, End Distance=1500, Minimum Output=2, Maximum Output=5, Clamp Output=true)
    )
    Index 0 Param 2=&Multiply_Float(
     A=1
     B=&FN_LOD_Distance_Remap(Start Distance=15000, End Distance=50000, Minimum Output=5, Maximum Output=2500, Clamp Output=true)
    )
    Index 1 Param 1=50
    Module.Index0Param0Write=true
    Module.Index0Param1Write=true
    Module.Index0Param2Write=true
    Module.Index0Param3Write=true
    Module.Index1Param0Write=true
    Module.Index1Param1Write=true
    Module.Index1Param2Write=true
    Module.Index1Param3Write=true
    Write Parameter Index 1=true
   )
   &CameraOffset(Camera Offset Amount=50)
   /SetVariables(Particles.Linear Color=&MultiplyLinearColor(Linear Color=$Particles.Color, Scale Factor=<15 15 15 1>)) {isSetParameters=true}
  ]
  rendererClasses [&NiagaraSpriteRendererProperties]
  renderers [{rendererClass=&NiagaraSpriteRendererProperties}]
 }
 FN_Sprite002 {
  simTarget=GPUComputeSim
  emitterUpdateScript [
   &EmitterState()
   &SpawnBurst_Instantaneous(Use Spawn Probability=false, UseLoopCountLimit=false)
  ]
  particleSpawnScript [
   &InitializeParticle(
    Color=&ColorFromCurve(
     ColorCurve=DataInterface() {propertyValues {
       AlphaCurve {keys [<at 0 1 RCIM_Cubic>,<at 0.9977 1 RCIM_Cubic>]}
       BlueCurve {keys [<at 0.1175 0 RCIM_Cubic>]}
       GreenCurve {keys [<at 0.1175 0.1279 RCIM_Cubic>]}
       RedCurve {keys [<at 0.1175 1 RCIM_Cubic>]}
      }}
     CurveIndex=&RandomRangeFloat(Minimum=0, Maximum=1)
    )
    Lifetime Max=3
    Lifetime Min=3
    Lifetime Mode=::NewEnumerator1
    Mesh Scale Mode=::NewEnumerator1
    Mesh Uniform Scale Max=250
    Mesh Uniform Scale Min=250
    Sprite Size Mode=::NewEnumerator1
    Uniform Sprite Size Max=300
    Uniform Sprite Size Min=300
    UsePositionOffset=false
    Write Material Random=true
   ) {module=InitializeParticle001}
   &AddVelocityFromPoint(
    Module.Use Falloff Distance=false
    Origin Offset=&UniformRangedVector(Minimum=<0 0 -100>, Maximum=<0 0 -100>)
    Origin Offset Coordinate Space=::Simulation
    Velocity Strength=&UniformRangedFloat(Minimum=10, Maximum=15)
   ) {enabled=false}
   &CurlNoiseForce(Noise Frequency=0.25, Noise Strength=5, Pan Noise=true, Pan Noise Field=<0 0 -5>) {enabled=false, module=CurlNoiseForce001}
   &ApplyInitialForces() {enabled=false}
   &InitialMeshOrientation(
    Mesh Axis To Orient=<0 0 0>
    Mesh Orientation Mode=::NewEnumerator2
    Orientation Vector=<0 0 1>
    Rotation=&RandomRangeVector(Minimum=<0 0 0>, Maximum=<0 0 0>)
    Use Rotation Vector=true
   )
   &ShapeLocation(
    Cylinder Height=2
    Cylinder Radius=&FN_LOD_Distance_Remap(Start Distance=5000, End Distance=15000, Minimum Output=6000, Maximum Output=3000, Clamp Output=true)
    MaskSpawnGroup=false
    Non Uniform Scale=<2 1 1>
    Offset=<0 0 1500>
    Offset Mode=::NewEnumerator0
    Override Seed=false
    Shape Primitive=::NewEnumerator1
    Surface Only=true
    Use Endcaps In Surface Only Mode=false
    UseOwnerScale=true
   ) {enabled=false}
  ]
  particleUpdateScript [
   &ParticleState()
   &UpdateMeshOrientation() {enabled=false, module=UpdateMeshOrientation001}
   &ScaleMeshSize(Scale Factor=&Multiply_VectorByFloat(
     Vector=<0.25 0.25 0.25>
     Float=&FN_LOD_Distance_Remap(Start Distance=5000, End Distance=50000, Minimum Output=0.5, Maximum Output=2, Clamp Output=true)
    ))
   &UpdateMeshOrientation(
    Facing Interpolation=false
    Orientation Method=::NewEnumerator1
    Override Quat=false
    TransformFacing=false
    TransformUp=false
    Up Direction:Vector3f=<0 0 -1>
   )
   &UpdateMeshOrientation(
    Rotation Rate=&RandomRangeFloat(Minimum=1, Maximum=2)
    Rotation Vector=&MakeVector(
     X=0
     Y=&Sine(Normalized Angle=$Emitter.Age, Period=&RandomRangeFloat(Minimum=1, Maximum=3), Scale=0.02, Bias=0)
     Z=0
    )
   ) {module=UpdateMeshOrientation002}
   &SolveForcesAndVelocity(Clamp Velocity=false, Limit Acceleration=false)
   &ScaleColor(
    Scale Alpha=&FloatFromCurve(FloatCurve=DataInterface() {propertyValues {Curve {keys [
         {arriveTangent=9.0725, interpMode=RCIM_Cubic, leaveTangent=9.0725, tangentMode=RCTM_User, time=0, value=0}
         {arriveTangent=-2.5249, interpMode=RCIM_Cubic, leaveTangent=-2.5249, tangentMode=RCTM_User, time=0.4, value=1.075}
         {arriveTangent=-0.0405, interpMode=RCIM_Cubic, leaveTangent=-0.0405, tangentMode=RCTM_User, time=1, value=0}
        ]}}})
    Scale RGB=<15 15 15>
    ScaleA=true
    ScaleRGB=true
   )
   &DynamicMaterialParameters(
    Index 0 Param 1=&Multiply_Float(
     A=25
     B=&FN_LOD_Distance_Remap(Start Distance=250, End Distance=1500, Minimum Output=2, Maximum Output=5, Clamp Output=true)
    )
    Index 0 Param 2=&Multiply_Float(
     A=100
     B=&FN_LOD_Distance_Remap(Start Distance=15000, End Distance=50000, Minimum Output=2, Maximum Output=5, Clamp Output=true)
    )
    Index 1 Param 1=50
    Module.Index0Param0Write=true
    Module.Index0Param1Write=true
    Module.Index0Param2Write=true
    Module.Index0Param3Write=true
    Module.Index1Param0Write=true
    Module.Index1Param1Write=true
    Module.Index1Param2Write=true
    Module.Index1Param3Write=true
    Write Parameter Index 1=true
   )
   &CameraOffset(Camera Offset Amount=50)
  ]
  rendererClasses [&NiagaraMeshRendererProperties]
  renderers [{rendererClass=&NiagaraMeshRendererProperties, rendererIndex=0}]
 }
}
summary {
 systemName=NS_NitroDrome_Reaction_OneSpot
 emitters [
  {emitterName=FN_Sprite001, simTarget=GPUComputeSim, rendererClasses [&NiagaraSpriteRendererProperties]}
  {bEnabled=false, emitterName=FN_Sprite, simTarget=GPUComputeSim, rendererClasses []}
  {emitterName=FN_Sprite002, simTarget=GPUComputeSim, rendererClasses [&NiagaraMeshRendererProperties]}
 ]
 userVariables []
}
