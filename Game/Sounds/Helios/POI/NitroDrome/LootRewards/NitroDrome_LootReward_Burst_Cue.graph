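# Sound cue for the NitroDrome loot reward burst, built from the POI's fireworks burst sounds and routed to the Athena ambient sound class.
# A Mixer layers two random picks at once, a 2D burst and a 3D burst, each through its own shared fireworks attenuation asset.
# Both pools are equally weighted, and the 2D pool preselects a subset of its variations at level load.
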SoundCue /Game/Sounds/Helios/POI/NitroDrome/LootRewards/NitroDrome_LootReward_Burst_Cue
root=Mixer_0
nodes [
 {
  id=Attenuation_0
  children [Random_4]
  properties {attenuationSettings=/Game/Sounds/Helios/POI/NitroDrome/Fireworks/NitroDrome_Fireworks_3D_Att}
 }
 {
  id=Attenuation_1
  children [Random_3]
  properties {attenuationSettings=/Game/Sounds/Helios/POI/NitroDrome/Fireworks/NitroDrome_Fireworks_2D_Att}
 }
 {id=Mixer_0, children [Attenuation_1, Attenuation_0], properties {inputVolume [1,1]}}
 {
  id=Random_3
  children [
   WavePlayer_20
   WavePlayer_21
   WavePlayer_22
   WavePlayer_23
   WavePlayer_24
   WavePlayer_25
   WavePlayer_26
   WavePlayer_27
   WavePlayer_28
   WavePlayer_29
   WavePlayer_30
   WavePlayer_31
   WavePlayer_32
   WavePlayer_33
   WavePlayer_34
  ]
  properties {preselectAtLevelLoad=5, weights [1,1,1,1,1,1,1,1,1,1,1,1,1,1,1]}
 }
 {
  id=Random_4
  children [
   WavePlayer_35
   WavePlayer_36
   WavePlayer_37
   WavePlayer_38
   WavePlayer_39
   WavePlayer_40
   WavePlayer_41
   WavePlayer_42
   WavePlayer_43
   WavePlayer_44
   WavePlayer_45
  ]
  properties {weights [1,1,1,1,1,1,1,1,1,1,1]}
 }
 {id=WavePlayer_20, properties {soundWaveAssetPtr=/Game/Sounds/Helios/POI/NitroDrome/Fireworks/NitroDrome_Fireworks_Burst_2D_01}}
 {id=WavePlayer_21, properties {soundWaveAssetPtr=/Game/Sounds/Helios/POI/NitroDrome/Fireworks/NitroDrome_Fireworks_Burst_2D_02}}
 {id=WavePlayer_22, properties {soundWaveAssetPtr=/Game/Sounds/Helios/POI/NitroDrome/Fireworks/NitroDrome_Fireworks_Burst_2D_03}}
 {id=WavePlayer_23, properties {soundWaveAssetPtr=/Game/Sounds/Helios/POI/NitroDrome/Fireworks/NitroDrome_Fireworks_Burst_2D_04}}
 {id=WavePlayer_24, properties {soundWaveAssetPtr=/Game/Sounds/Helios/POI/NitroDrome/Fireworks/NitroDrome_Fireworks_Burst_2D_05}}
 {id=WavePlayer_25, properties {soundWaveAssetPtr=/Game/Sounds/Helios/POI/NitroDrome/Fireworks/NitroDrome_Fireworks_Burst_2D_06}}
 {id=WavePlayer_26, properties {soundWaveAssetPtr=/Game/Sounds/Helios/POI/NitroDrome/Fireworks/NitroDrome_Fireworks_Burst_2D_07}}
 {id=WavePlayer_27, properties {soundWaveAssetPtr=/Game/Sounds/Helios/POI/NitroDrome/Fireworks/NitroDrome_Fireworks_Burst_2D_08}}
 {id=WavePlayer_28, properties {soundWaveAssetPtr=/Game/Sounds/Helios/POI/NitroDrome/Fireworks/NitroDrome_Fireworks_Burst_2D_09}}
 {id=WavePlayer_29, properties {soundWaveAssetPtr=/Game/Sounds/Helios/POI/NitroDrome/Fireworks/NitroDrome_Fireworks_Burst_2D_10}}
 {id=WavePlayer_30, properties {soundWaveAssetPtr=/Game/Sounds/Helios/POI/NitroDrome/Fireworks/NitroDrome_Fireworks_Burst_2D_11}}
 {id=WavePlayer_31, properties {soundWaveAssetPtr=/Game/Sounds/Helios/POI/NitroDrome/Fireworks/NitroDrome_Fireworks_Burst_2D_12}}
 {id=WavePlayer_32, properties {soundWaveAssetPtr=/Game/Sounds/Helios/POI/NitroDrome/Fireworks/NitroDrome_Fireworks_Burst_2D_13}}
 {id=WavePlayer_33, properties {soundWaveAssetPtr=/Game/Sounds/Helios/POI/NitroDrome/Fireworks/NitroDrome_Fireworks_Burst_2D_14}}
 {id=WavePlayer_34, properties {soundWaveAssetPtr=/Game/Sounds/Helios/POI/NitroDrome/Fireworks/NitroDrome_Fireworks_Burst_2D_15}}
 {id=WavePlayer_35, properties {soundWaveAssetPtr=/Game/Sounds/Helios/POI/NitroDrome/Fireworks/NitroDrome_Fireworks_Burst_3D_01}}
 {id=WavePlayer_36, properties {soundWaveAssetPtr=/Game/Sounds/Helios/POI/NitroDrome/Fireworks/NitroDrome_Fireworks_Burst_3D_02}}
 {id=WavePlayer_37, properties {soundWaveAssetPtr=/Game/Sounds/Helios/POI/NitroDrome/Fireworks/NitroDrome_Fireworks_Burst_3D_03}}
 {id=WavePlayer_38, properties {soundWaveAssetPtr=/Game/Sounds/Helios/POI/NitroDrome/Fireworks/NitroDrome_Fireworks_Burst_3D_04}}
 {id=WavePlayer_39, properties {soundWaveAssetPtr=/Game/Sounds/Helios/POI/NitroDrome/Fireworks/NitroDrome_Fireworks_Burst_3D_05}}
 {id=WavePlayer_40, properties {soundWaveAssetPtr=/Game/Sounds/Helios/POI/NitroDrome/Fireworks/NitroDrome_Fireworks_Burst_3D_06}}
 {id=WavePlayer_41, properties {soundWaveAssetPtr=/Game/Sounds/Helios/POI/NitroDrome/Fireworks/NitroDrome_Fireworks_Burst_3D_07}}
 {id=WavePlayer_42, properties {soundWaveAssetPtr=/Game/Sounds/Helios/POI/NitroDrome/Fireworks/NitroDrome_Fireworks_Burst_3D_08}}
 {id=WavePlayer_43, properties {soundWaveAssetPtr=/Game/Sounds/Helios/POI/NitroDrome/Fireworks/NitroDrome_Fireworks_Burst_3D_09}}
 {id=WavePlayer_44, properties {soundWaveAssetPtr=/Game/Sounds/Helios/POI/NitroDrome/Fireworks/NitroDrome_Fireworks_Burst_3D_10}}
 {id=WavePlayer_45, properties {soundWaveAssetPtr=/Game/Sounds/Helios/POI/NitroDrome/Fireworks/NitroDrome_Fireworks_Burst_3D_11}}
]
properties {maxDistance=20000, soundClassObject=&Fort_Ambient_Athena, volumeMultiplier=0.752}
