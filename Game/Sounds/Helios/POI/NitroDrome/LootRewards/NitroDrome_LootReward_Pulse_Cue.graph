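# Sound cue for the NitroDrome loot reward pulse, playing a quiet spark one-shot as ambient audio.
# A Random node picks among eight equally weighted EntranceHead spark waves borrowed from the same POI.
# A quality-level root routes the first two levels to that Random node and leaves the other two empty.
# The cue is spatialized with the NitroDrome fireworks 3D attenuation and uses the Athena ambient sound class.
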SoundCue /Game/Sounds/Helios/POI/NitroDrome/LootRewards/NitroDrome_LootReward_Pulse_Cue
root=QualityLevel_0
nodes [
 {id=QualityLevel_0, children [Random_0,Random_0,null,null]}
 {
  id=Random_0
  children [WavePlayer_0,WavePlayer_1,WavePlayer_2,WavePlayer_3,WavePlayer_4,WavePlayer_5,WavePlayer_6,WavePlayer_7]
  properties {bShouldExcludeFromBranchCulling=true, weights [1,1,1,1,1,1,1,1]}
 }
 {id=WavePlayer_0, properties {soundWaveAssetPtr=/Game/Sounds/Helios/POI/NitroDrome/EntranceHead/NitroDrome_EntranceHead_Sparks_01}}
 {id=WavePlayer_1, properties {soundWaveAssetPtr=/Game/Sounds/Helios/POI/NitroDrome/EntranceHead/NitroDrome_EntranceHead_Sparks_02}}
 {id=WavePlayer_2, properties {soundWaveAssetPtr=/Game/Sounds/Helios/POI/NitroDrome/EntranceHead/NitroDrome_EntranceHead_Sparks_03}}
 {id=WavePlayer_3, properties {soundWaveAssetPtr=/Game/Sounds/Helios/POI/NitroDrome/EntranceHead/NitroDrome_EntranceHead_Sparks_04}}
 {id=WavePlayer_4, properties {soundWaveAssetPtr=/Game/Sounds/Helios/POI/NitroDrome/EntranceHead/NitroDrome_EntranceHead_Sparks_05}}
 {id=WavePlayer_5, properties {soundWaveAssetPtr=/Game/Sounds/Helios/POI/NitroDrome/EntranceHead/NitroDrome_EntranceHead_Sparks_06}}
 {id=WavePlayer_6, properties {soundWaveAssetPtr=/Game/Sounds/Helios/POI/NitroDrome/EntranceHead/NitroDrome_EntranceHead_Sparks_07}}
 {id=WavePlayer_7, properties {soundWaveAssetPtr=/Game/Sounds/Helios/POI/NitroDrome/EntranceHead/NitroDrome_EntranceHead_Sparks_08}}
]
properties {
 attenuationSettings=/Game/Sounds/Helios/POI/NitroDrome/Fireworks/NitroDrome_Fireworks_3D_Att
 maxDistance=20000
 soundClassObject=&Fort_Ambient_Athena
 volumeMultiplier=0.1
}
